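# Physics asset for the F_MED_BrightShimmer female character body cosmetic, covering its dangling accessories: a pouch with zipper and a pair of back ties.
# The pouch uses four capsules and is constrained to a parent sphere, with the zipper chained to it under tight angular limits.
# Left and right tie capsules hang from a base sphere with wider swing than twist.
# Unconstrained pelvis, spine_01 and right thigh capsules provide nearby body collision shapes.
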PhysicsAsset /BRCosmetics/Characters/Player/Female/Medium/Bodies/F_MED_BrightShimmer/Meshes/F_MED_BrightShimmer_Physics
bodies {
 dyn_pouch [
  {
   shapeType=Capsule
   length=9.15
   radius=2.15
   shapeName=dyn_pouch_capsule
   center=<6.082 -10.9661 6.0989>
   rotation=<rot -76.2726 -112.8681 124.0146>
  }
  {shapeType=Capsule, length=5, radius=2.25, shapeName=dyn_pouch_capsule, center=<13.7858 9.1023 2.5321>, rotation=<rot -75 -0.0003 0.0018>}
  {shapeType=Capsule, length=11.4, radius=2.1, shapeName=dyn_pouch_capsule, center=<17.7251 1.8121 2.2439>, rotation=<rot 15 0.0015 -99.9993>}
  {
   shapeType=Capsule
   length=6.75
   radius=1.75
   shapeName=dyn_pouch_capsule
   center=<-8.8904 -7.5362 11.6824>
   rotation=<rot 49.3509 -41.8579 -13.695>
  }
 ]
 dyn_pouch_parent [{shapeType=Sphere, radius=3, shapeName=dyn_pouch_parent_sphere}]
 dyn_ties_back_base [{shapeType=Sphere, radius=1.25, shapeName=dyn_ties_back_base_sphere}]
 dyn_ties_back_l [{shapeType=Capsule, length=2.2, radius=1, shapeName=dyn_ties_back_l_capsule, center=<2.1288 0 0.4969>, rotation=<rot -90 179.9998 -179.9991>}]
 dyn_ties_back_parent [{shapeType=Sphere, radius=1, shapeName=dyn_ties_back_parent_sphere}]
 dyn_ties_back_r [{
   shapeType=Capsule
   length=2.2
   radius=1
   shapeName=dyn_ties_back_r_capsule
   center=<1.9106 0 -0.1132>
   rotation=<rot -75 -179.9999 -179.9992>
  }]
 dyn_zipp [{shapeType=Capsule, length=2.75, radius=0.5, shapeName=dyn_zipp_capsule, center=<1.8751 0 0.0845>, rotation=<rot -90 157.2398 -147.2398>}]
 pelvis [{shapeType=Capsule, length=11.4357, radius=8.616, center=<-0.3109 -1.7381 1.2841>, rotation=<rot -3.5309 -85.5195 -1.69>}]
 spine_01 [{shapeType=Capsule, length=10.5, radius=7, shapeName=spine_01_capsule, center=<1.6012 0 0>}]
 thigh_r [{shapeType=Capsule, length=32.75, radius=8, shapeName=thigh_r_capsule, center=<18.9451 -0.0001 0.0428>, rotation=<rot -90 0.0009 -0.0021>}]
}
constraints [
 {bone1Name=dyn_ties_back_r, bone2Name=dyn_ties_back_base, swing1LimitDegrees=45, swing2LimitDegrees=45, twistLimitDegrees=5}
 {bone1Name=dyn_ties_back_l, bone2Name=dyn_ties_back_base, swing1LimitDegrees=45, swing2LimitDegrees=45, twistLimitDegrees=5}
 {bone1Name=dyn_pouch, bone2Name=dyn_pouch_parent, swing1LimitDegrees=45, swing2LimitDegrees=45, twistLimitDegrees=1}
 {bone1Name=dyn_zipp, bone2Name=dyn_pouch, swing1LimitDegrees=5, swing2LimitDegrees=3, twistLimitDegrees=3}
]
