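# Physics asset for the bow mesh of the female medium BrightShimmer cosmetic body, covering its dynamic front, neck and hand ties.
# Small spheres form base and parent bodies, while capsules cover tie ends, bow loops, clavicles and upper spine.
# Swing and twist-limited constraints link ties to their bases, tightest at the front and loosest at the hand.
# Only the front base is constrained to its parent; spine and clavicle bodies have no constraints.
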PhysicsAsset /BRCosmetics/Characters/Player/Female/Medium/Bodies/F_MED_BrightShimmer/Meshes/F_MED_BrightShimmer_bow_Physics
bodies {
 clavicle_l [{shapeType=Capsule, length=7.85, radius=4.35, shapeName=clavicle_l_capsule, center=<5.336 -1.7224 0.4977>, rotation=<rot -90 90 -90.0005>}]
 clavicle_r [{
   shapeType=Capsule
   length=7.85
   radius=4.35
   shapeName=clavicle_r_capsule
   center=<-5.4552 2.5308 -0.5653>
   rotation=<rot -90 0.0002 -0.0008>
  }]
 dyn_ties_front_base [{shapeType=Sphere, radius=0.5, shapeName=dyn_ties_front_base_sphere}]
 dyn_ties_front_bow_l [{shapeType=Capsule, length=1.25, radius=0.75, shapeName=dyn_ties_front_bow_l_capsule, center=<0.6971 0 0>, rotation=<rot 85 0.0003 -0.0002>}]
 dyn_ties_front_bow_r [{
   shapeType=Capsule
   length=1.25
   radius=0.75
   shapeName=dyn_ties_front_bow_r_capsule
   center=<0.592 0 0.1697>
   rotation=<rot -75.0003 -0.0005 0.0012>
  }]
 dyn_ties_front_l [{shapeType=Capsule, length=1.25, radius=0.75, shapeName=dyn_ties_front_l_capsule, rotation=<rot -25 -0.0001 20.0004>}]
 dyn_ties_front_parent [{shapeType=Sphere, radius=0.5, shapeName=dyn_ties_front_parent_sphere}]
 dyn_ties_front_r [{shapeType=Capsule, length=1.25, radius=0.75, shapeName=dyn_ties_front_r_capsule, center=<0 -0.1757 0>, rotation=<rot 80 -0.0002 -5.0006>}]
 dyn_ties_hand_base [{shapeType=Sphere, radius=1.25, shapeName=dyn_ties_hand_base_sphere}]
 dyn_ties_hand_down [{shapeType=Capsule, length=4.25, radius=1.25, shapeName=dyn_ties_hand_down_capsule, center=<1.3707 0 0>, rotation=<rot 90 180 -179.9998>}]
 dyn_ties_hand_up [{shapeType=Capsule, length=4.25, radius=1, shapeName=dyn_ties_hand_up_capsule, center=<1.1702 0 0>, rotation=<rot 0.0001 89.9998 89.9993>}]
 dyn_ties_neck_base [{shapeType=Sphere, radius=0.5, shapeName=dyn_ties_neck_base_sphere}]
 dyn_ties_neck_bow_l [{
   shapeType=Capsule
   length=2
   radius=1
   shapeName=dyn_ties_neck_bow_l_capsule
   center=<1.1782 0 0.3157>
   rotation=<rot 75.0002 -179.9996 -179.9998>
  }]
 dyn_ties_neck_bow_r [{shapeType=Capsule, length=2, radius=1, shapeName=dyn_ties_neck_bow_r_capsule, center=<1.4064 0 0>, rotation=<rot -90 180 -179.9998>}]
 dyn_ties_neck_l [{shapeType=Capsule, length=4.4, radius=0.6, shapeName=dyn_ties_neck_l_capsule, center=<2.7284 -0.8948 0.6213>, rotation=<rot -80 0 -9.9992>}]
 dyn_ties_neck_parent [{shapeType=Sphere, radius=0.5, shapeName=dyn_ties_neck_parent_sphere}]
 dyn_ties_neck_r [{
   shapeType=Capsule
   length=4.4
   radius=0.6
   shapeName=dyn_ties_neck_r_capsule
   center=<2.1623 -1.1353 -0.6145>
   rotation=<rot -70.0143 166.2747 -176.2154>
  }]
 spine_04 [{shapeType=Capsule, length=15, radius=5.5, shapeName=spine_04_capsule, center=<-0.3694 7.1491 0>}]
 spine_05 [{shapeType=Capsule, length=10.95, radius=5.05, shapeName=spine_05_capsule, center=<6.4675 -1.4063 0.0633>, rotation=<rot 0 -0.0001 0>}]
}
constraints [
 {bone1Name=dyn_ties_neck_bow_l, bone2Name=dyn_ties_neck_base, swing1LimitDegrees=45, swing2LimitDegrees=45, twistLimitDegrees=20}
 {bone1Name=dyn_ties_neck_bow_r, bone2Name=dyn_ties_neck_base, swing1LimitDegrees=45, swing2LimitDegrees=45, twistLimitDegrees=20}
 {bone1Name=dyn_ties_front_base, bone2Name=dyn_ties_front_parent, swing1LimitDegrees=20, swing2LimitDegrees=15, twistLimitDegrees=0.5}
 {bone1Name=dyn_ties_front_bow_l, bone2Name=dyn_ties_front_base, swing1LimitDegrees=25, swing2LimitDegrees=30, twistLimitDegrees=2}
 {bone1Name=dyn_ties_front_bow_r, bone2Name=dyn_ties_front_base, swing1LimitDegrees=25, swing2LimitDegrees=30, twistLimitDegrees=2}
 {bone1Name=dyn_ties_front_l, bone2Name=dyn_ties_front_base, swing1LimitDegrees=25, swing2LimitDegrees=30, twistLimitDegrees=2}
 {bone1Name=dyn_ties_front_r, bone2Name=dyn_ties_front_base, swing1LimitDegrees=25, swing2LimitDegrees=30, twistLimitDegrees=2}
 {bone1Name=dyn_ties_hand_up, bone2Name=dyn_ties_hand_base, swing1LimitDegrees=75, swing2LimitDegrees=75, twistLimitDegrees=30}
 {bone1Name=dyn_ties_hand_down, bone2Name=dyn_ties_hand_base, swing1LimitDegrees=75, swing2LimitDegrees=75, twistLimitDegrees=30}
 {bone1Name=dyn_ties_neck_l, bone2Name=dyn_ties_neck_base, swing1LimitDegrees=45, swing2LimitDegrees=45, twistLimitDegrees=25}
 {bone1Name=dyn_ties_neck_r, bone2Name=dyn_ties_neck_base, swing1LimitDegrees=45, swing2LimitDegrees=45, twistLimitDegrees=25}
]
