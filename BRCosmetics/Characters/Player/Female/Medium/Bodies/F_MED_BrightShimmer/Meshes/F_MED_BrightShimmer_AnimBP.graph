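# Animation Blueprint for the F_MED_BrightShimmer female body cosmetic, adding secondary motion to its ties, pouches, carabiner and belt.
# It copies the attached parent mesh's pose through a crouch-aware deform Control Rig, then applies additive bone offsets, some gated by crouch, fall or rigid-body flags.
# Constraints and two RigidBody simulations, based at pelvis and neck, refine the result.
# Each update switches pouch, carabiner, belt and simulation-alpha values by crouch state.
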AnimBlueprint /BRCosmetics/Characters/Player/Female/Medium/Bodies/F_MED_BrightShimmer/Meshes/F_MED_BrightShimmer_AnimBP
target_skeleton=&/F_MED_BrightShimmer_Skeleton
graphs [
 AnimGraph {nodes [
   n1 = Control Rig!(Source=Copy Pose From Mesh() {bUseAttachedParent=true}, Alpha, Crouch:bool=Get bIsCrouching()) -> (Pose) {
    alphaBoolBlend {blendOption=Linear}
    controlRigAssetReference {blueprintRigClass=&/F_MED_BrightShimmer_ControlRig_Deform.F_MED_BrightShimmer_ControlRig_Deform_C}
   }
   n2 = "Transform (Modify) Bone - Bone: dyn_ties_back_base"!(
    ComponentPose=Local To Component(LocalPose=@n1)
    bAlphaBoolEnabled:bool=Get bIsFalling()
    Translation="0,0,0"
    Rotation=Get Rotator_back_ties() -> (Rotator_back_ties:Rotator)
    Scale="1,1,1"
   ) -> (Pose) {
    alphaInputType=Bool
    rotationMode=BMM_Additive
    rotationSpace=BCS_BoneSpace
    translationMode=BMM_Additive
    translationSpace=BCS_BoneSpace
    alphaBoolBlend {blendInTime=0.45, blendOption=Linear, blendOutTime=0.45}
    boneToModify {boneName=dyn_ties_back_base}
   }
   n3 = Get bIsCrouching()
   n4 = "Transform (Modify) Bone - Bone: dyn_ties_back_l"!(ComponentPose=@n2, bAlphaBoolEnabled:bool=@n3, Translation="0,7,0", Rotation="0,-15,0", Scale="1,1,1") -> (Pose) {
    alphaInputType=Bool
    rotationMode=BMM_Additive
    rotationSpace=BCS_BoneSpace
    translationMode=BMM_Additive
    translationSpace=BCS_BoneSpace
    alphaBoolBlend {blendInTime=0.45, blendOption=Linear, blendOutTime=0.45}
    boneToModify {boneName=dyn_ties_back_l}
    rotation {yaw=-15}
    translation {y=7}
   }
   n5 = "Transform (Modify) Bone - Bone: dyn_ties_back_r"!(ComponentPose=@n4, bAlphaBoolEnabled:bool=@n3, Translation="0,7,0", Rotation="0,0,0", Scale="1,1,1") -> (Pose) {
    alphaInputType=Bool
    rotationMode=BMM_Additive
    rotationSpace=BCS_BoneSpace
    translationMode=BMM_Additive
    translationSpace=BCS_BoneSpace
    alphaBoolBlend {blendInTime=1, blendOption=Linear, blendOutTime=1}
    boneToModify {boneName=dyn_ties_back_r}
    translation {y=7}
   }
   n6 = "Transform (Modify) Bone - Bone: dyn_ties_hand_parent"!(ComponentPose=@n5, bAlphaBoolEnabled:bool=Get bIsFalling(), Translation="0,0,0", Rotation="0,-60,0", Scale="1,1,1") -> (Pose) {
    alphaInputType=Bool
    rotationMode=BMM_Additive
    rotationSpace=BCS_BoneSpace
    translationMode=BMM_Additive
    translationSpace=BCS_BoneSpace
    alphaBoolBlend {blendInTime=0.25, blendOption=Linear, blendOutTime=0.25}
    boneToModify {boneName=dyn_ties_hand_parent}
    rotation {yaw=-60}
   }
   n7 = "Transform (Modify) Bone - Bone: dyn_ties_neck_base"!(ComponentPose=@n6, bAlphaBoolEnabled:bool=True, Translation="0,0,0", Rotation="0,5,0", Scale="1,1,1") -> (Pose) {
    alphaInputType=Bool
    rotationMode=BMM_Additive
    rotationSpace=BCS_BoneSpace
    translationMode=BMM_Additive
    translationSpace=BCS_BoneSpace
    alphaBoolBlend {blendInTime=0.25, blendOption=Linear, blendOutTime=0.25}
    boneToModify {boneName=dyn_ties_neck_base}
    rotation {yaw=5}
   }
   n8 = "Transform (Modify) Bone - Bone: dyn_pouch_parent"!(
    ComponentPose=@n7
    bAlphaBoolEnabled:bool=Get bRigidBodyEnabled()
    Translation=Get Translation_pouch() -> (Translation_pouch:Vector)
    Rotation=Get Rotator_Pouch() -> (Rotator_Pouch:Rotator)
    Scale="1,1,1"
   ) -> (Pose) {
    alphaInputType=Bool
    rotationMode=BMM_Additive
    rotationSpace=BCS_BoneSpace
    translationMode=BMM_Additive
    translationSpace=BCS_BoneSpace
    alphaBoolBlend {blendInTime=10, blendOption=Linear, blendOutTime=10}
    alphaScaleBiasClamp {bInterpResult=true, interpSpeedDecreasing=1, interpSpeedIncreasing=1}
    boneToModify {boneName=dyn_pouch_parent}
   }
   n9 = "Transform (Modify) Bone - Bone: dyn_carabine"(ComponentPose=@n8, Alpha=1, Translation="0,0,0", Rotation=Get Rotator_Carabine() -> (Rotator_Carabine:Rotator), Scale="1,1,1") {
    rotationMode=BMM_Additive
    rotationSpace=BCS_BoneSpace
    translationMode=BMM_Additive
    translationSpace=BCS_BoneSpace
    alphaBoolBlend {blendOption=Linear}
    alphaScaleBiasClamp {bInterpResult=true}
    boneToModify {boneName=dyn_carabine}
   }
   n10 = "Transform (Modify) Bone - Bone: dyn_pouch_back"(
    ComponentPose=@n9
    Alpha=1
    Translation=Get Translation_pouch_back() -> (Translation_pouch_back:Vector)
    Rotation=Get Rotator_Pouch_Back() -> (Rotator_Pouch_Back:Rotator)
    Scale="1,1,1"
   ) {
    rotationMode=BMM_Additive
    rotationSpace=BCS_BoneSpace
    translationMode=BMM_Additive
    translationSpace=BCS_BoneSpace
    alphaBoolBlend {blendOption=Linear}
    alphaScaleBiasClamp {bInterpResult=true, interpSpeedIncreasing=50}
    boneToModify {boneName=dyn_pouch_back}
   }
   n11 = "Constraint - dyn_belt_front (P)"(ComponentPose=@n10, Alpha=Get Alfa_belt_constraint() -> (Alfa_belt_constraint:double), ConstraintWeights_0=1) {
    alphaBoolBlend {blendOption=Linear}
    alphaScaleBiasClamp {bInterpResult=true}
    boneToModify {boneName=dyn_belt_front}
    constraintSetup [{targetBone {boneName=dyn_pouch}}]
    constraintWeights [1]
   }
   n12 = RigidBody!(ComponentPose=@n11, Alpha=Get Alfa_pouch_rigidbody() -> (Alfa_pouch_rigidbody:double), OverrideWorldGravity:Vector="0,0,-980", ExternalForce) -> (Pose) {
    bClampLinearTranslationLimitToRefPose=true
    bOverrideWorldGravity=true
    overridePhysicsAsset=&/F_MED_BrightShimmer_Physics
    simulationSpace=BaseBoneSpace
    alphaBoolBlend {blendOption=Linear}
    alphaScaleBiasClamp {bInterpResult=true}
    baseBoneRef {boneName=pelvis}
    componentLinearAccScale=<1 1 1>
    componentLinearVelScale=<1 1 1>
    overrideWorldGravity {z=-980}
   }
   n13 = RigidBody!(ComponentPose=@n12, Alpha, ExternalForce) -> (Pose) {
    bClampLinearTranslationLimitToRefPose=true
    bOverrideWorldGravity=true
    cachedBoundsScale=1
    overridePhysicsAsset=&/F_MED_BrightShimmer_bow_Physics
    simulationSpace=BaseBoneSpace
    alphaBoolBlend {blendOption=Linear}
    alphaScaleBiasClamp {bInterpResult=true, interpSpeedDecreasing=0.5, interpSpeedIncreasing=0.5}
    baseBoneRef {boneName=neck_01}
    componentAppliedLinearAccClamp=<2000 2000 2000>
    componentLinearAccScale=<1 1 1>
    componentLinearVelScale=<1 1 1>
    overrideWorldGravity {z=-980}
    simSpaceSettings {maxAngularAcceleration=1000, maxAngularVelocity=100, maxLinearAcceleration=1000, maxLinearVelocity=100}
   }
   n14 = "Constraint - dyn_ties_front_base (P)"(ComponentPose=@n13, Alpha=1, ConstraintWeights_0=1) {
    alphaBoolBlend {blendOption=Linear}
    alphaScaleBiasClamp {bInterpResult=true, interpSpeedDecreasing=1, interpSpeedIncreasing=1}
    boneToModify {boneName=dyn_ties_front_base}
    constraintSetup [{targetBone {boneName=spine_04}}]
    constraintWeights [1]
   }
   n15 = "Constraint - dyn_ties_hand_up (P)"(ComponentPose=@n14, Alpha=0.5, ConstraintWeights_0=1) {
    alphaBoolBlend {blendOption=Linear}
    alphaScaleBiasClamp {bInterpResult=true, interpSpeedDecreasing=1, interpSpeedIncreasing=1}
    boneToModify {boneName=dyn_ties_hand_up}
    constraintSetup [{targetBone {boneName=dyn_ties_hand_base}}]
    constraintWeights [1]
   }
   n16 = "Constraint - dyn_ties_hand_down (P)"(ComponentPose=@n15, Alpha=0.33, ConstraintWeights_0=1) {
    alphaBoolBlend {blendOption=Linear}
    alphaScaleBiasClamp {bInterpResult=true, interpSpeedDecreasing=1, interpSpeedIncreasing=1}
    boneToModify {boneName=dyn_ties_hand_down}
    constraintSetup [{targetBone {boneName=dyn_ties_hand_base}}]
    constraintWeights [1]
   }
   Output Pose(Result=Component To Local(ComponentPose=@n16)) {layerGroup=None}
  ]}
 EventGraph {nodes [
   Event BlueprintUpdateAnimation()
   > n1 = Branch(Condition=Get bIsCrouching())
   > Set Alfa_pouch_rigidbody(Alfa_pouch_rigidbody:double=1) -> (Output_Get:double)
   > Set Translation_pouch(Translation_pouch:Vector="-3,0,0") -> (Output_Get:Vector)
   > Set Rotator_Pouch(Rotator_Pouch:Rotator="10,10,10") -> (Output_Get:Rotator)
   > Set Rotator_Carabine(Rotator_Carabine:Rotator="-30,-10,20") -> (Output_Get:Rotator)
   > Set Rotator_Pouch_Back(Rotator_Pouch_Back:Rotator="-10,20,0") -> (Output_Get:Rotator)
   > Set Alfa_belt_constraint(Alfa_belt_constraint:double=0.5) -> (Output_Get:double)
   Set Alfa_pouch_rigidbody(execute=@n1.else, Alfa_pouch_rigidbody=0.75)
   > Set Translation_pouch(Translation_pouch="-1,0,0")
   > Set Rotator_Pouch(Rotator_Pouch="-10,0,0")
   > Set Translation_pouch_back(Translation_pouch_back:Vector="0,0,1") -> (Output_Get:Vector)
   > Set Rotator_Carabine(Rotator_Carabine="-20,20,0")
   > Set Rotator_Pouch_Back(Rotator_Pouch_Back="0, 0, 0")
   > Set Alfa_belt_constraint(Alfa_belt_constraint=0)
  ]}
]
variables [
 Alfa_belt_constraint {type=number, default=1}
 Alfa_pouch_rigidbody {type=number, default=0.75}
 Overrite_World_Gravity {type=Vector}
 Rotator_back_ties {type=Rotator}
 Rotator_Carabine {type=Rotator}
 Rotator_hand_ties {type=Rotator}
 Rotator_Pouch {type=Rotator, default=<rot -10 20 -5>}
 Rotator_Pouch_Back {type=Rotator, default=<rot -10 20 -5>}
 Translation_pouch {type=Vector}
 Translation_pouch_back {type=Vector}
]
