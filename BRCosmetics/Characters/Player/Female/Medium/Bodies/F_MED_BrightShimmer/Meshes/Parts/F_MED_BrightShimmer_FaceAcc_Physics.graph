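# Physics asset for the BrightShimmer medium female body's FaceAcc part, setting up physics for twin ponytails and two hair strands.
# Each side has a four-bone ponytail chain and a two-bone strand chain of capsules, constrained link by link back to the head.
# Ponytail swing limits widen from base to tip.
# Head, neck, upper spine, clavicle and upper-arm capsules have no parent constraints, apparently serving as body colliders for the hair.
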PhysicsAsset /BRCosmetics/Characters/Player/Female/Medium/Bodies/F_MED_BrightShimmer/Meshes/Parts/F_MED_BrightShimmer_FaceAcc_Physics
bodies {
 clavicle_l [{shapeType=Capsule, length=9.75, radius=3.75, shapeName=clavicle_l_capsule, center=<6.8278 -1.5232 -1.0354>, rotation=<rot -90 0 -0.0003>}]
 clavicle_r [{shapeType=Capsule, length=9.75, radius=3.75, shapeName=clavicle_r_capsule, center=<-7.4397 1.5232 1.0354>, rotation=<rot -90 0.158 -0.1583>}]
 dyn_hair_strend_1_l [{shapeType=Capsule, length=6, radius=1.5, shapeName=dyn_hair_strend_1_l_capsule, center=<4.6065 0 -0.3221>, rotation=<rot -90 -90 90.0002>}]
 dyn_hair_strend_1_r [{
   shapeType=Capsule
   length=5
   radius=1.75
   shapeName=dyn_hair_strend_1_r_capsule
   center=<-3.9536 0.4165 -0.1811>
   rotation=<rot -84.9998 -0.0004 0.0006>
  }]
 dyn_hair_strend_base_l [{
   shapeType=Capsule
   length=1.5
   radius=1.75
   shapeName=dyn_hair_strend_base_l_capsule
   center=<2.6143 -0.2975 0.2136>
   rotation=<rot -90 0.0049 -0.0049>
  }]
 dyn_hair_strend_base_r [{
   shapeType=Capsule
   length=2
   radius=1.5
   shapeName=dyn_hair_strend_base_r_capsule
   center=<-2.4141 0.4508 -0.3011>
   rotation=<rot -90 179.9604 -179.9604>
  }]
 dyn_ponytail_1_l [{shapeType=Capsule, length=2.75, radius=2.25, shapeName=dyn_ponytail_1_l_capsule, center=<3.6603 0 0>, rotation=<rot -90 -179.9852 179.9852>}]
 dyn_ponytail_1_r [{shapeType=Capsule, length=2.75, radius=2.25, shapeName=dyn_ponytail_1_r_capsule, center=<-3.433 0 0>, rotation=<rot -90 -179.9604 179.9605>}]
 dyn_ponytail_2_l [{shapeType=Capsule, length=3.25, radius=1.75, shapeName=dyn_ponytail_2_l_capsule, center=<4.3902 0 0>, rotation=<rot -90 0 0.0001>}]
 dyn_ponytail_2_r [{shapeType=Capsule, length=3.25, radius=1.75, shapeName=dyn_ponytail_2_r_capsule, center=<-4.1697 0 0>, rotation=<rot -90 -179.3669 179.3671>}]
 dyn_ponytail_3_l [{
   shapeType=Capsule
   length=4.25
   radius=2
   shapeName=dyn_ponytail_3_l_capsule
   center=<4.1013 0.1033 -0.111>
   rotation=<rot -85.0001 -179.9998 179.9999>
  }]
 dyn_ponytail_3_r [{
   shapeType=Capsule
   length=4.25
   radius=2
   shapeName=dyn_ponytail_3_r_capsule
   center=<-3.9583 -0.1033 0.341>
   rotation=<rot -85.0001 -179.9998 179.9999>
  }]
 dyn_ponytail_base_l [{
   shapeType=Capsule
   length=1.25
   radius=3.75
   shapeName=dyn_ponytail_base_l_capsule
   center=<3.3251 0 0>
   rotation=<rot -90 160.5288 -160.529>
  }]
 dyn_ponytail_base_r [{
   shapeType=Capsule
   length=1.25
   radius=3.75
   shapeName=dyn_ponytail_base_r_capsule
   center=<-3.0335 0 0>
   rotation=<rot -90 179.9999 179.9999>
  }]
 head [
  {shapeType=Capsule, length=2.25, radius=7.373, center=<4.8013 -0.2578 -0.4968>, rotation=<rot 3.0087 2.3123 -107.328>}
  {shapeType=Capsule, length=4.25, radius=4, shapeName=head_capsule, center=<-2.2175 1.0417 0>, rotation=<rot -90 180 -90.0005>}
 ]
 neck_01 [{
   shapeType=Capsule
   length=6
   radius=4.5
   shapeName=neck_01_capsule
   center=<2.4905 0.0374 -0.2214>
   rotation=<rot -83.2992 140.0097 -139.3438>
  }]
 spine_04 [{shapeType=Capsule, length=1, radius=12, shapeName=spine_04_capsule, center=<0.0318 2.0059 0>}]
 spine_05 [{shapeType=Capsule, length=10, radius=7.5, shapeName=spine_05_capsule, center=<0.0206 -0.0205 0>}]
 upperarm_l [{
   shapeType=Capsule
   length=14.25
   radius=3.75
   shapeName=upperarm_r_capsule
   center=<10.2635 -0.2039 -0.9266>
   rotation=<rot -85 179.9967 0.0036>
  }]
 upperarm_r [{
   shapeType=Capsule
   length=14.25
   radius=3.75
   shapeName=upperarm_r_capsule
   center=<-10.2712 0.2039 0.8389>
   rotation=<rot -85 179.9967 -179.9964>
  }]
}
constraints [
 {bone1Name=dyn_ponytail_base_l, bone2Name=head, swing1LimitDegrees=15, swing2LimitDegrees=15, twistLimitDegrees=10}
 {bone1Name=dyn_ponytail_base_r, bone2Name=head, swing1LimitDegrees=15, swing2LimitDegrees=15, twistLimitDegrees=10}
 {bone1Name=dyn_hair_strend_base_r, bone2Name=head, swing1LimitDegrees=45, swing2LimitDegrees=45, twistLimitDegrees=25}
 {bone1Name=dyn_hair_strend_base_l, bone2Name=head, swing1LimitDegrees=45, swing2LimitDegrees=45, twistLimitDegrees=25}
 {bone1Name=dyn_ponytail_1_l, bone2Name=dyn_ponytail_base_l, swing1LimitDegrees=25, swing2LimitDegrees=25, twistLimitDegrees=20}
 {bone1Name=dyn_ponytail_2_l, bone2Name=dyn_ponytail_1_l, swing1LimitDegrees=25, swing2LimitDegrees=45, twistLimitDegrees=15}
 {bone1Name=dyn_ponytail_3_l, bone2Name=dyn_ponytail_2_l, swing1LimitDegrees=45, swing2LimitDegrees=60, twistLimitDegrees=20}
 {bone1Name=dyn_ponytail_1_r, bone2Name=dyn_ponytail_base_r, swing1LimitDegrees=25, swing2LimitDegrees=25, twistLimitDegrees=20}
 {bone1Name=dyn_ponytail_2_r, bone2Name=dyn_ponytail_1_r, swing1LimitDegrees=25, swing2LimitDegrees=45, twistLimitDegrees=15}
 {bone1Name=dyn_ponytail_3_r, bone2Name=dyn_ponytail_2_r, swing1LimitDegrees=45, swing2LimitDegrees=60, twistLimitDegrees=20}
 {bone1Name=dyn_hair_strend_1_l, bone2Name=dyn_hair_strend_base_l, swing1LimitDegrees=45, swing2LimitDegrees=45, twistLimitDegrees=30}
 {bone1Name=dyn_hair_strend_1_r, bone2Name=dyn_hair_strend_base_r, swing1LimitDegrees=45, swing2LimitDegrees=45, twistLimitDegrees=30}
]
